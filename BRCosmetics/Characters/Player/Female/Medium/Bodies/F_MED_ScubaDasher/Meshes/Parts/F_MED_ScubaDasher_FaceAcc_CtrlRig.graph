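# Control Rig for the ScubaDasher female medium body's face accessory part that corrects a hair bone as the neck and head move.
# Three spherical pose readers track neck_01, neck_02 and head, each weighting a following Modify Transforms that applies the same rotation to dyn_hairs_10_01.
# The chain starts with an identity-transform modify on head, and all three readers have debug drawing enabled.
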ControlRigBlueprint /BRCosmetics/Characters/Player/Female/Medium/Bodies/F_MED_ScubaDasher/Meshes/Parts/F_MED_ScubaDasher_FaceAcc_CtrlRig
graphs [
 RigVMFunctionLibrary {links [], nodes []}
 RigVMModel {nodes [
   RigUnit_BeginExecution = &RigUnit_BeginExecution()
   RigVMFunction_Sequence = &RigVMFunction_Sequence(ExecuteContext=@RigUnit_BeginExecution.ExecutePin)
   &RigUnit_ModifyTransforms(
    ItemToModify=((Item=(Type=Bone,Name="head"),Transform=(Rotation=(X=0,Y=0,Z=0,W=1),Translation=(X=0,Y=0,Z=0),Scale3D=(X=1,Y=1,Z=1))))
    Weight=1
    WeightMinimum=0
    ExecutePin=@RigVMFunction_Sequence.A
   )
   > n1 = &RigUnit_SphericalPoseReader(
    ActiveRegionScaleFactors=(PositiveWidth=1,NegativeWidth=1,PositiveHeight=1,NegativeHeight=1)
    ActiveRegionSize=0.1
    Debug=(bDrawDebug=true,bDraw2D=False,bDrawLocalAxes=False,DebugScale=25,DebugSegments=20,DebugThickness=0.25)
    DriverAxis=(X=1,Y=0,Z=0)
    DriverItem=(Type=Bone,Name="neck_01")
    FalloffRegionScaleFactors=(PositiveWidth=1,NegativeWidth=1,PositiveHeight=1,NegativeHeight=1)
    FalloffSize=0.5
    FlipHeightScaling=False
    FlipWidthScaling=False
    OptionalParentItem=(Type=Bone,Name="None")
    RotationOffset=(X=-90,Y=0,Z=0)
   )
   > &RigUnit_ModifyTransforms(
    ItemToModify=((Item=(Type=Bone,Name="dyn_hairs_10_01"),Transform=(Rotation=(X=0,Y=-0.5,Z=0,W=0.866),Translation=(X=0,Y=0,Z=0),Scale3D=(X=1,Y=1,Z=1))))
    WeightMinimum=0
    Weight=@n1.OutputParam
   )
   > n2 = &RigUnit_SphericalPoseReader(
    ActiveRegionScaleFactors=(PositiveWidth=1,NegativeWidth=1,PositiveHeight=1,NegativeHeight=1)
    ActiveRegionSize=0.1
    Debug=(bDrawDebug=true,bDraw2D=False,bDrawLocalAxes=False,DebugScale=25,DebugSegments=20,DebugThickness=0.25)
    DriverAxis=(X=1,Y=0,Z=0)
    DriverItem=(Type=Bone,Name="neck_02")
    FalloffRegionScaleFactors=(PositiveWidth=1,NegativeWidth=1,PositiveHeight=1,NegativeHeight=1)
    FalloffSize=0.5
    FlipHeightScaling=False
    FlipWidthScaling=False
    OptionalParentItem=(Type=Bone,Name="None")
    RotationOffset=(X=-90,Y=0,Z=0)
   )
   > &RigUnit_ModifyTransforms(
    ItemToModify=((Item=(Type=Bone,Name="dyn_hairs_10_01"),Transform=(Rotation=(X=0,Y=-0.5,Z=0,W=0.866),Translation=(X=0,Y=0,Z=0),Scale3D=(X=1,Y=1,Z=1))))
    WeightMinimum=0
    Weight=@n2.OutputParam
   )
   > n3 = &RigUnit_SphericalPoseReader(
    ActiveRegionScaleFactors=(PositiveWidth=1,NegativeWidth=1,PositiveHeight=1,NegativeHeight=1)
    ActiveRegionSize=0.1
    Debug=(bDrawDebug=true,bDraw2D=False,bDrawLocalAxes=False,DebugScale=25,DebugSegments=20,DebugThickness=0.25)
    DriverAxis=(X=1,Y=0,Z=0)
    DriverItem=(Type=Bone,Name="head")
    FalloffRegionScaleFactors=(PositiveWidth=1,NegativeWidth=1,PositiveHeight=1,NegativeHeight=1)
    FalloffSize=0.5
    FlipHeightScaling=False
    FlipWidthScaling=False
    OptionalParentItem=(Type=Bone,Name="None")
    RotationOffset=(X=-90,Y=0,Z=0)
   )
   > &RigUnit_ModifyTransforms(
    ItemToModify=((Item=(Type=Bone,Name="dyn_hairs_10_01"),Transform=(Rotation=(X=0,Y=-0.5,Z=0,W=0.866),Translation=(X=0,Y=0,Z=0),Scale3D=(X=1,Y=1,Z=1))))
    WeightMinimum=0
    Weight=@n3.OutputParam
   )
  ]}
]
hierarchy [
 root {loc=<0 0 0>, rot=<0 0 0 1>}
 pelvis {parent=root, loc=<0 -0.5301 98.0571>, rot=<0.0127 0.707 -0.0127 -0.707>}
 spine_01 {parent=pelvis, loc=<6.1208 -0.9328 0>, rot=<0 0 0.034 -0.9994>}
 spine_02 {parent=spine_01, loc=<6.5997 -0.2095 0>, rot=<0 0 0.0238 -0.9997>}
 spine_03 {parent=spine_02, loc=<6.4914 -0.1237 0>, rot=<0 0 0.0149 -0.9999>}
 spine_04 {parent=spine_03, loc=<6.9598 0.0794 0>, rot=<0 0 -0.0132 -0.9999>}
 spine_05 {parent=spine_04, loc=<6.9987 0 0>, rot=<0 0 -0.0015 -1>}
 neck_01 {parent=spine_05, loc=<11.3138 0.1529 0>, rot=<0 0 -0.2003 -0.9797>}
 neck_02 {parent=neck_01, loc=<4.5627 0 0>, rot=<0 0 0.0504 -0.9987>}
 head {parent=neck_02, loc=<5.3914 0 0>, rot=<0 0 0.115 -0.9934>}
 dyn_hairs_01 {parent=head, loc=<4.9607 5.1409 6.0555>, rot=<0.1831 -0.6738 0.7013 -0.1437>}
 dyn_hairs_02 {parent=head, loc=<2.2144 1.6509 6.5451>, rot=<0.0223 -0.6898 0.7165 -0.101>}
 dyn_hairs_18 {parent=head, loc=<7.5178 9.5986 4.6018>, rot=<0.0065 0.6349 0.7591 0.1434>}
 dyn_hairs_19 {parent=head, loc=<8.2606 10.1397 1.1782>, rot=<-0.1197 0.6499 0.7497 0.0359>}
 dyn_hairs_20 {parent=head, loc=<8.7087 8.9638 -2.0236>, rot=<0.1071 -0.8263 -0.5136 -0.2048>}
 dyn_hairs_13_01 {parent=head, loc=<10.7618 -4.741 0.3705>, rot=<0.2342 -0.5818 0.7343 -0.2598>}
 dyn_hairs_13_02 {parent=dyn_hairs_13_01, loc=<4.9289 0 0>, rot=<0 0 0 1>}
 dyn_hairs_12_01 {parent=head, loc=<1.9957 -3.1899 -4.1488>, rot=<0.1463 -0.9243 0.3424 -0.0838>}
 dyn_hairs_12_02 {parent=dyn_hairs_12_01, loc=<2.7873 0 0>, rot=<0 0 0 1>}
 dyn_hairs_11_01 {parent=head, loc=<4.1969 -3.4578 4.5081>, rot=<0.0702 -0.2758 0.9575 -0.0476>}
 dyn_hairs_11_02 {parent=dyn_hairs_11_01, loc=<3.5541 0 0>, rot=<0 0 0 1>}
 dyn_hairs_10_01 {parent=head, loc=<1.7747 -5.3074 -0.8645>, rot=<0.0606 -0.7025 0.7088 -0.0187>}
 dyn_hairs_10_02 {parent=dyn_hairs_10_01, loc=<4.2391 0 0>, rot=<0 0 0 1>}
 dyn_hairs_09_01 {parent=head, loc=<5.3442 -5.1109 -0.8768>, rot=<0.2012 -0.6642 0.6834 -0.2263>}
 dyn_hairs_09_02 {parent=dyn_hairs_09_01, loc=<4.2391 0 0>, rot=<0 0 0 1>}
 dyn_hairs_08_01 {parent=head, loc=<5.0898 -1.9072 4.4139>, rot=<0.1735 -0.2898 0.8838 -0.3237>}
 dyn_hairs_08_02 {parent=dyn_hairs_08_01, loc=<4.9289 0 0>, rot=<0 0 0 1>}
 dyn_hairs_06_01 {parent=head, loc=<7.8692 -2.4489 -7.6214>, rot=<0.4027 -0.8426 0.297 -0.199>}
 dyn_hairs_06_02 {parent=dyn_hairs_06_01, loc=<4.3644 0 0>, rot=<0 0 0 1>}
 dyn_hairs_07_01 {parent=head, loc=<5.9016 0.819 -6.3867>, rot=<0.4427 -0.8954 0.0472 -0.0003>}
 dyn_hairs_07_02 {parent=dyn_hairs_07_01, loc=<5.1366 0 0>, rot=<0 0 0 1>}
 dyn_hairs_05_01 {parent=head, loc=<9.094 -4.1875 -4.0548>, rot=<0.2905 -0.667 0.6258 -0.2812>}
 dyn_hairs_05_02 {parent=dyn_hairs_05_01, loc=<3.915 0 0>, rot=<0 0 0 1>}
 dyn_hairs_04_01 {parent=head, loc=<8.3678 -4.0926 4.7963>, rot=<0.2352 -0.341 0.8772 -0.2426>}
 dyn_hairs_04_02 {parent=dyn_hairs_04_01, loc=<4.5346 0 0>, rot=<0 0 0 1>}
 dyn_hairs_03_01 {parent=head, loc=<7.5192 0.7885 6.3136>, rot=<0.228 -0.3108 0.8612 -0.3313>}
 dyn_hairs_03_02 {parent=dyn_hairs_03_01, loc=<4.9289 0 0>, rot=<0 0 0 1>}
]
