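# Animation Blueprint for the female medium ScubaDasher Goggles face accessory, adding secondary motion to its hair bones.
# It copies the attached parent pose through the shared ScubaDasher FaceAcc Control Rig, then chains sprung AnimDynamics per bone, some using spine_05 spherical limits.
# Two hair-tip bones get additive roll when bIsInFrontEnd is true.
# Each update stores Gravity Override from the male (M_MED) parameter asset, though no visible AnimGraph node reads it.
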AnimBlueprint /BRCosmetics/Characters/Player/Female/Medium/Bodies/F_MED_ScubaDasher_Goggles/Meshes/Parts/F_MED_ScubaDasher_Goggles_FaceAcc_AnimBP
target_skeleton=&/F_MED_ScubaDasher_Goggles_FaceAcc_Skeleton
graphs [
 AnimGraph {nodes [
   n1 = Control Rig(Source=Copy Pose From Mesh() {bUseAttachedParent=true}) {
    alphaBoolBlend {blendOption=Linear}
    controlRigAssetReference {blueprintRigClass=/BRCosmetics/Characters/Player/Female/Medium/Bodies/F_MED_ScubaDasher/Meshes/Parts/F_MED_ScubaDasher_FaceAcc_CtrlRig.F_MED_ScubaDasher_FaceAcc_CtrlRig_C}
   }
   n2 = Anim Dynamics - Bone: dyn_hairs_01!(ComponentPose=Local To Component(LocalPose=@n1), Alpha=1, GravityOverride:Vector="0,0,-980", ExternalForce="0,0,0") -> (Pose) {
    angularDampingOverride=0.9999
    angularSpringConstant=10000
    bAngularSpring=true
    bOverrideAngularDamping=true
    bUseGravityOverride=true
    gravityScale=0.35
    simulationSpace=RootRelative
    alphaBoolBlend {blendOption=Linear}
    boundBone {boneName=dyn_hairs_01}
    componentAppliedLinearAccClamp=<2000 2000 2000>
    componentLinearAccScale=<2 2 2>
    componentLinearVelScale=<2 2 2>
    gravityOverride {z=-980}
    physicsBodyDefinitions [{
      boundBone {boneName=dyn_hairs_01}
      boxExtents=<35 35 20>
      constraintSetup {angularLimitsMax=<0 10 10>, angularLimitsMin=<0 -10 -10>, angularTarget=<1 0 0>}
      localJointOffset=<4 0 0>
     }]
   }
   n3 = Anim Dynamics - Bone: dyn_hairs_02!(ComponentPose=@n2, Alpha=1, GravityOverride:Vector="0,0,-980", ExternalForce="0,0,0") -> (Pose) {
    angularDampingOverride=0.9999
    angularSpringConstant=10000
    bAngularSpring=true
    bOverrideAngularDamping=true
    bUseGravityOverride=true
    gravityScale=0.35
    simulationSpace=RootRelative
    alphaBoolBlend {blendOption=Linear}
    boundBone {boneName=dyn_hairs_02}
    componentAppliedLinearAccClamp=<2000 2000 2000>
    componentLinearAccScale=<2 2 2>
    componentLinearVelScale=<2 2 2>
    gravityOverride {z=-980}
    physicsBodyDefinitions [{
      boundBone {boneName=dyn_hairs_02}
      boxExtents=<35 35 20>
      constraintSetup {angularLimitsMax=<0 10 10>, angularLimitsMin=<0 -10 -10>, angularTarget=<1 0 0>}
      localJointOffset=<4 0 0>
     }]
   }
   n4 = Anim Dynamics - Bone: dyn_hairs_03_01(ComponentPose=@n3, Alpha=1, ExternalForce="0,0,0") {
    angularDampingOverride=0.9999
    angularSpringConstant=4000
    bAngularSpring=true
    bOverrideAngularDamping=true
    bUseGravityOverride=true
    gravityScale=0.35
    simulationSpace=RootRelative
    alphaBoolBlend {blendOption=Linear}
    boundBone {boneName=dyn_hairs_03_01}
    componentAppliedLinearAccClamp=<2000 2000 2000>
    componentLinearAccScale=<2 2 2>
    componentLinearVelScale=<2 2 2>
    gravityOverride {z=-2000}
    physicsBodyDefinitions [{
      boundBone {boneName=dyn_hairs_03_01}
      boxExtents=<25 25 15>
      constraintSetup {angularLimitsMax=<0 2 2>, angularLimitsMin=<0 -2 -2>, angularTarget=<1 0 0>}
      localJointOffset=<7 0 0>
     }]
   }
   n5 = Anim Dynamics - Bone: dyn_hairs_03_02(ComponentPose=@n4, Alpha=1, ExternalForce="0,0,0") {
    angularDampingOverride=0.9999
    angularSpringConstant=4000
    bAngularSpring=true
    bOverrideAngularDamping=true
    bUseGravityOverride=true
    gravityScale=0.35
    simulationSpace=RootRelative
    alphaBoolBlend {blendOption=Linear}
    boundBone {boneName=dyn_hairs_03_02}
    componentAppliedLinearAccClamp=<2000 2000 2000>
    componentLinearAccScale=<2 2 2>
    componentLinearVelScale=<2 2 2>
    gravityOverride {z=-2000}
    physicsBodyDefinitions [{
      boundBone {boneName=dyn_hairs_03_02}
      boxExtents=<25 25 15>
      constraintSetup {angularLimitsMax=<0 1 1>, angularLimitsMin=<0 -1 -1>, angularTarget=<1 0 0>}
      localJointOffset=<7 0 0>
     }]
   }
   n6 = Anim Dynamics - Bone: dyn_hairs_08_01(ComponentPose=@n5, Alpha=1, ExternalForce="0,0,0") {
    angularDampingOverride=0.9999
    angularSpringConstant=4000
    bAngularSpring=true
    bOverrideAngularDamping=true
    bUseGravityOverride=true
    gravityScale=0.35
    simulationSpace=RootRelative
    alphaBoolBlend {blendOption=Linear}
    boundBone {boneName=dyn_hairs_08_01}
    componentAppliedLinearAccClamp=<2000 2000 2000>
    componentLinearAccScale=<2 2 2>
    componentLinearVelScale=<2 2 2>
    gravityOverride {z=-2000}
    physicsBodyDefinitions [{
      boundBone {boneName=dyn_hairs_08_01}
      boxExtents=<25 25 15>
      constraintSetup {angularLimitsMax=<0 2 2>, angularLimitsMin=<0 -2 -2>, angularTarget=<1 0 0>}
      localJointOffset=<7 0 0>
     }]
   }
   n7 = Anim Dynamics - Bone: dyn_hairs_08_02(ComponentPose=@n6, Alpha=1, ExternalForce="0,0,0") {
    angularDampingOverride=0.9999
    angularSpringConstant=4000
    bAngularSpring=true
    bOverrideAngularDamping=true
    bUseGravityOverride=true
    gravityScale=0.35
    simulationSpace=RootRelative
    alphaBoolBlend {blendOption=Linear}
    boundBone {boneName=dyn_hairs_08_02}
    componentAppliedLinearAccClamp=<2000 2000 2000>
    componentLinearAccScale=<2 2 2>
    componentLinearVelScale=<2 2 2>
    gravityOverride {z=-2000}
    physicsBodyDefinitions [{
      boundBone {boneName=dyn_hairs_08_02}
      boxExtents=<25 25 15>
      constraintSetup {angularLimitsMax=<0 1 1>, angularLimitsMin=<0 -1 -1>, angularTarget=<1 0 0>}
      localJointOffset=<7 0 0>
     }]
   }
   n8 = Anim Dynamics - Bone: dyn_hairs_11_01(ComponentPose=@n7, Alpha=1, ExternalForce="0,0,0") {
    angularDampingOverride=0.9999
    angularSpringConstant=4000
    bAngularSpring=true
    bOverrideAngularDamping=true
    bUseGravityOverride=true
    gravityScale=0.35
    simulationSpace=RootRelative
    alphaBoolBlend {blendOption=Linear}
    boundBone {boneName=dyn_hairs_11_01}
    componentAppliedLinearAccClamp=<2000 2000 2000>
    componentLinearAccScale=<2 2 2>
    componentLinearVelScale=<2 2 2>
    gravityOverride {z=-2000}
    physicsBodyDefinitions [{
      boundBone {boneName=dyn_hairs_11_01}
      boxExtents=<25 25 15>
      constraintSetup {angularLimitsMax=<0 2 2>, angularLimitsMin=<0 -2 -2>, angularTarget=<1 0 0>}
      localJointOffset=<7 0 0>
     }]
   }
   n9 = Anim Dynamics - Bone: dyn_hairs_11_02(ComponentPose=@n8, Alpha=1, ExternalForce="0,0,0") {
    angularDampingOverride=0.9999
    angularSpringConstant=4000
    bAngularSpring=true
    bOverrideAngularDamping=true
    bUseGravityOverride=true
    bUseSphericalLimits=true
    gravityScale=0.35
    simulationSpace=RootRelative
    alphaBoolBlend {blendOption=Linear}
    boundBone {boneName=dyn_hairs_11_02}
    componentAppliedLinearAccClamp=<2000 2000 2000>
    componentLinearAccScale=<2 2 2>
    componentLinearVelScale=<2 2 2>
    gravityOverride {z=-2000}
    physicsBodyDefinitions [{
      boundBone {boneName=dyn_hairs_11_02}
      boxExtents=<25 25 15>
      constraintSetup {angularLimitsMax=<0 1 1>, angularLimitsMin=<0 -1 -1>, angularTarget=<1 0 0>}
      localJointOffset=<7 0 0>
     }]
    sphericalLimits [{limitRadius=7, drivingBone {boneName=spine_05}, sphereLocalOffset=<10 4 4>}]
   }
   n10 = Anim Dynamics - Bone: dyn_hairs_04_01(ComponentPose=@n9, Alpha=1, ExternalForce="0,0,0") {
    angularDampingOverride=0.9999
    angularSpringConstant=4000
    bAngularSpring=true
    bOverrideAngularDamping=true
    bUseGravityOverride=true
    gravityScale=0.35
    simulationSpace=RootRelative
    alphaBoolBlend {blendOption=Linear}
    boundBone {boneName=dyn_hairs_04_01}
    componentAppliedLinearAccClamp=<2000 2000 2000>
    componentLinearAccScale=<2 2 2>
    componentLinearVelScale=<2 2 2>
    gravityOverride {z=-2000}
    physicsBodyDefinitions [{
      boundBone {boneName=dyn_hairs_04_01}
      boxExtents=<25 25 15>
      constraintSetup {angularLimitsMax=<0 2 2>, angularLimitsMin=<0 -2 -2>, angularTarget=<1 0 0>}
      localJointOffset=<7 0 0>
     }]
   }
   n11 = Anim Dynamics - Bone: dyn_hairs_04_02(ComponentPose=@n10, Alpha=1, ExternalForce="0,0,0") {
    angularDampingOverride=0.9999
    angularSpringConstant=4000
    bAngularSpring=true
    bOverrideAngularDamping=true
    bUseGravityOverride=true
    gravityScale=0.35
    simulationSpace=RootRelative
    alphaBoolBlend {blendOption=Linear}
    boundBone {boneName=dyn_hairs_04_02}
    componentAppliedLinearAccClamp=<2000 2000 2000>
    componentLinearAccScale=<2 2 2>
    componentLinearVelScale=<2 2 2>
    gravityOverride {z=-2000}
    physicsBodyDefinitions [{
      boundBone {boneName=dyn_hairs_04_02}
      boxExtents=<25 25 15>
      constraintSetup {angularLimitsMax=<0 1 1>, angularLimitsMin=<0 -1 -1>, angularTarget=<1 0 0>}
      localJointOffset=<7 0 0>
     }]
   }
   n12 = Anim Dynamics - Bone: dyn_hairs_10_01(ComponentPose=@n11, Alpha=1, ExternalForce="0,0,0") {
    angularDampingOverride=0.9999
    angularSpringConstant=4000
    bAngularSpring=true
    bOverrideAngularDamping=true
    bUseGravityOverride=true
    bUseSphericalLimits=true
    gravityScale=0.35
    simulationSpace=RootRelative
    alphaBoolBlend {blendOption=Linear}
    boundBone {boneName=dyn_hairs_10_01}
    componentAppliedLinearAccClamp=<2000 2000 2000>
    gravityOverride {z=-2000}
    physicsBodyDefinitions [{
      boundBone {boneName=dyn_hairs_10_01}
      boxExtents=<25 25 15>
      constraintSetup {angularLimitsMax=<0 0 0>, angularLimitsMin=<0 0 0>, angularTarget=<1 0 0>}
      localJointOffset=<1 0 0>
     }]
    sphericalLimits [{limitRadius=0, drivingBone {boneName=None}, sphereLocalOffset=<0 0 0>}]
   }
   n13 = Anim Dynamics - Bone: dyn_hairs_10_02(ComponentPose=@n12, Alpha=1, ExternalForce="0,0,0") {
    angularDampingOverride=0.9999
    angularSpringConstant=4000
    bAngularSpring=true
    bOverrideAngularDamping=true
    bUseGravityOverride=true
    bUseSphericalLimits=true
    gravityScale=0.35
    simulationSpace=RootRelative
    alphaBoolBlend {blendOption=Linear}
    boundBone {boneName=dyn_hairs_10_02}
    componentAppliedLinearAccClamp=<2000 2000 2000>
    gravityOverride {z=-2000}
    physicsBodyDefinitions [{
      boundBone {boneName=dyn_hairs_10_02}
      boxExtents=<25 25 15>
      constraintSetup {angularLimitsMax=<0 0 0>, angularLimitsMin=<0 0 0>, angularTarget=<1 0 0>}
      localJointOffset=<1 0 0>
     }]
    sphericalLimits [{limitRadius=8, drivingBone {boneName=spine_05}, sphereLocalOffset=<10 4 0>}]
   }
   n14 = Anim Dynamics - Bone: dyn_hairs_12_01(ComponentPose=@n13, Alpha=1, ExternalForce="0,0,0") {
    angularDampingOverride=0.9999
    angularSpringConstant=4000
    bAngularSpring=true
    bOverrideAngularDamping=true
    bUseGravityOverride=true
    gravityScale=0.35
    simulationSpace=RootRelative
    alphaBoolBlend {blendOption=Linear}
    boundBone {boneName=dyn_hairs_12_01}
    componentAppliedLinearAccClamp=<2000 2000 2000>
    componentLinearAccScale=<2 2 2>
    componentLinearVelScale=<2 2 2>
    gravityOverride {z=-2000}
    physicsBodyDefinitions [{
      boundBone {boneName=dyn_hairs_12_01}
      boxExtents=<25 25 15>
      constraintSetup {angularLimitsMax=<0 2 2>, angularLimitsMin=<0 -2 -2>, angularTarget=<1 0 0>}
      localJointOffset=<7 0 0>
     }]
   }
   n15 = Anim Dynamics - Bone: dyn_hairs_12_02(ComponentPose=@n14, Alpha=1, ExternalForce="0,0,0") {
    angularDampingOverride=0.9999
    angularSpringConstant=4000
    bAngularSpring=true
    bOverrideAngularDamping=true
    bUseGravityOverride=true
    bUseSphericalLimits=true
    gravityScale=0.35
    simulationSpace=RootRelative
    alphaBoolBlend {blendOption=Linear}
    boundBone {boneName=dyn_hairs_12_02}
    componentAppliedLinearAccClamp=<2000 2000 2000>
    componentLinearAccScale=<2 2 2>
    componentLinearVelScale=<2 2 2>
    gravityOverride {z=-2000}
    physicsBodyDefinitions [{
      boundBone {boneName=dyn_hairs_12_02}
      boxExtents=<25 25 15>
      constraintSetup {angularLimitsMax=<0 1 1>, angularLimitsMin=<0 -1 -1>, angularTarget=<1 0 0>}
      localJointOffset=<7 0 0>
     }]
    sphericalLimits [{limitRadius=8, drivingBone {boneName=spine_05}, sphereLocalOffset=<10 4 0>}]
   }
   n16 = Anim Dynamics - Bone: dyn_hairs_07_01(ComponentPose=@n15, Alpha=1, ExternalForce="0,0,0") {
    angularDampingOverride=0.9999
    angularSpringConstant=4000
    bAngularSpring=true
    bOverrideAngularDamping=true
    bUseGravityOverride=true
    gravityScale=0.35
    simulationSpace=RootRelative
    alphaBoolBlend {blendOption=Linear}
    boundBone {boneName=dyn_hairs_07_01}
    componentAppliedLinearAccClamp=<2000 2000 2000>
    componentLinearAccScale=<2 2 2>
    componentLinearVelScale=<2 2 2>
    gravityOverride {z=-2000}
    physicsBodyDefinitions [{
      boundBone {boneName=dyn_hairs_07_01}
      boxExtents=<25 25 15>
      constraintSetup {angularLimitsMax=<0 2 2>, angularLimitsMin=<0 -2 -2>, angularTarget=<1 0 0>}
      localJointOffset=<7 0 0>
     }]
   }
   n17 = Anim Dynamics - Bone: dyn_hairs_07_02(ComponentPose=@n16, Alpha=1, ExternalForce="0,0,0") {
    angularDampingOverride=0.9999
    angularSpringConstant=4000
    bAngularSpring=true
    bOverrideAngularDamping=true
    bUseGravityOverride=true
    gravityScale=0.35
    simulationSpace=RootRelative
    alphaBoolBlend {blendOption=Linear}
    boundBone {boneName=dyn_hairs_07_02}
    componentAppliedLinearAccClamp=<2000 2000 2000>
    componentLinearAccScale=<2 2 2>
    componentLinearVelScale=<2 2 2>
    gravityOverride {z=-2000}
    physicsBodyDefinitions [{
      boundBone {boneName=dyn_hairs_07_02}
      boxExtents=<25 25 15>
      constraintSetup {angularLimitsMax=<0 1 1>, angularLimitsMin=<0 -1 -1>, angularTarget=<1 0 0>}
      localJointOffset=<7 0 0>
     }]
   }
   n18 = Anim Dynamics - Bone: dyn_hairs_09_01(ComponentPose=@n17, Alpha=1, ExternalForce="0,0,0") {
    angularDampingOverride=0.9999
    angularSpringConstant=4000
    bAngularSpring=true
    bOverrideAngularDamping=true
    bUseGravityOverride=true
    gravityScale=0.35
    simulationSpace=RootRelative
    alphaBoolBlend {blendOption=Linear}
    boundBone {boneName=dyn_hairs_09_01}
    componentAppliedLinearAccClamp=<2000 2000 2000>
    componentLinearAccScale=<2 2 2>
    componentLinearVelScale=<2 2 2>
    gravityOverride {z=-2000}
    physicsBodyDefinitions [{
      boundBone {boneName=dyn_hairs_09_01}
      boxExtents=<25 25 15>
      constraintSetup {angularLimitsMax=<0 2 2>, angularLimitsMin=<0 -2 -2>, angularTarget=<1 0 0>}
      localJointOffset=<7 0 0>
     }]
   }
   n19 = Anim Dynamics - Bone: dyn_hairs_09_02(ComponentPose=@n18, Alpha=1, ExternalForce="0,0,0") {
    angularDampingOverride=0.9999
    angularSpringConstant=4000
    bAngularSpring=true
    bOverrideAngularDamping=true
    bUseGravityOverride=true
    gravityScale=0.35
    simulationSpace=RootRelative
    alphaBoolBlend {blendOption=Linear}
    boundBone {boneName=dyn_hairs_09_02}
    componentAppliedLinearAccClamp=<2000 2000 2000>
    componentLinearAccScale=<2 2 2>
    componentLinearVelScale=<2 2 2>
    gravityOverride {z=-2000}
    physicsBodyDefinitions [{
      boundBone {boneName=dyn_hairs_09_02}
      boxExtents=<25 25 15>
      constraintSetup {angularLimitsMax=<0 1 1>, angularLimitsMin=<0 -1 -1>, angularTarget=<1 0 0>}
      localJointOffset=<7 0 0>
     }]
   }
   n20 = Anim Dynamics - Bone: dyn_hairs_13_01(ComponentPose=@n19, Alpha=1, ExternalForce="0,0,0") {
    angularDampingOverride=0.9999
    angularSpringConstant=4000
    bAngularSpring=true
    bOverrideAngularDamping=true
    bUseGravityOverride=true
    gravityScale=0.35
    simulationSpace=RootRelative
    alphaBoolBlend {blendOption=Linear}
    boundBone {boneName=dyn_hairs_13_01}
    componentAppliedLinearAccClamp=<2000 2000 2000>
    componentLinearAccScale=<2 2 2>
    componentLinearVelScale=<2 2 2>
    gravityOverride {z=-2000}
    physicsBodyDefinitions [{
      boundBone {boneName=dyn_hairs_13_01}
      boxExtents=<25 25 15>
      constraintSetup {angularLimitsMax=<0 2 2>, angularLimitsMin=<0 -2 -2>, angularTarget=<1 0 0>}
      localJointOffset=<7 0 0>
     }]
   }
   n21 = Anim Dynamics - Bone: dyn_hairs_13_02(ComponentPose=@n20, Alpha=1, ExternalForce="0,0,0") {
    angularDampingOverride=0.9999
    angularSpringConstant=4000
    bAngularSpring=true
    bOverrideAngularDamping=true
    bUseGravityOverride=true
    gravityScale=0.35
    simulationSpace=RootRelative
    alphaBoolBlend {blendOption=Linear}
    boundBone {boneName=dyn_hairs_13_02}
    componentAppliedLinearAccClamp=<2000 2000 2000>
    componentLinearAccScale=<2 2 2>
    componentLinearVelScale=<2 2 2>
    gravityOverride {z=-2000}
    physicsBodyDefinitions [{
      boundBone {boneName=dyn_hairs_13_02}
      boxExtents=<25 25 15>
      constraintSetup {angularLimitsMax=<0 1 1>, angularLimitsMin=<0 -1 -1>, angularTarget=<1 0 0>}
      localJointOffset=<7 0 0>
     }]
   }
   n22 = Anim Dynamics - Bone: dyn_hairs_05_01(ComponentPose=@n21, Alpha=1, ExternalForce="0,0,0") {
    angularDampingOverride=0.9999
    angularSpringConstant=4000
    bAngularSpring=true
    bOverrideAngularDamping=true
    bUseGravityOverride=true
    gravityScale=0.35
    simulationSpace=RootRelative
    alphaBoolBlend {blendOption=Linear}
    boundBone {boneName=dyn_hairs_05_01}
    componentAppliedLinearAccClamp=<2000 2000 2000>
    componentLinearAccScale=<2 2 2>
    componentLinearVelScale=<2 2 2>
    gravityOverride {z=-2000}
    physicsBodyDefinitions [{
      boundBone {boneName=dyn_hairs_05_01}
      boxExtents=<25 25 15>
      constraintSetup {angularLimitsMax=<0 2 2>, angularLimitsMin=<0 -2 -2>, angularTarget=<1 0 0>}
      localJointOffset=<7 0 0>
     }]
   }
   n23 = Anim Dynamics - Bone: dyn_hairs_05_02(ComponentPose=@n22, Alpha=1, ExternalForce="0,0,0") {
    angularDampingOverride=0.9999
    angularSpringConstant=4000
    bAngularSpring=true
    bOverrideAngularDamping=true
    bUseGravityOverride=true
    gravityScale=0.35
    simulationSpace=RootRelative
    alphaBoolBlend {blendOption=Linear}
    boundBone {boneName=dyn_hairs_05_02}
    componentAppliedLinearAccClamp=<2000 2000 2000>
    componentLinearAccScale=<2 2 2>
    componentLinearVelScale=<2 2 2>
    gravityOverride {z=-2000}
    physicsBodyDefinitions [{
      boundBone {boneName=dyn_hairs_05_02}
      boxExtents=<25 25 15>
      constraintSetup {angularLimitsMax=<0 1 1>, angularLimitsMin=<0 -1 -1>, angularTarget=<1 0 0>}
      localJointOffset=<7 0 0>
     }]
   }
   n24 = Anim Dynamics - Bone: dyn_hairs_06_01(ComponentPose=@n23, Alpha=1, ExternalForce="0,0,0") {
    angularDampingOverride=0.9999
    angularSpringConstant=4000
    bAngularSpring=true
    bOverrideAngularDamping=true
    bUseGravityOverride=true
    gravityScale=0.35
    simulationSpace=RootRelative
    alphaBoolBlend {blendOption=Linear}
    boundBone {boneName=dyn_hairs_06_01}
    componentAppliedLinearAccClamp=<2000 2000 2000>
    componentLinearAccScale=<2 2 2>
    componentLinearVelScale=<2 2 2>
    gravityOverride {z=-2000}
    physicsBodyDefinitions [{
      boundBone {boneName=dyn_hairs_06_01}
      boxExtents=<25 25 15>
      constraintSetup {angularLimitsMax=<0 2 2>, angularLimitsMin=<0 -2 -2>, angularTarget=<1 0 0>}
      localJointOffset=<7 0 0>
     }]
   }
   n25 = Anim Dynamics - Bone: dyn_hairs_06_02(ComponentPose=@n24, Alpha=1, ExternalForce="0,0,0") {
    angularDampingOverride=0.9999
    angularSpringConstant=4000
    bAngularSpring=true
    bOverrideAngularDamping=true
    bUseGravityOverride=true
    gravityScale=0.35
    simulationSpace=RootRelative
    alphaBoolBlend {blendOption=Linear}
    boundBone {boneName=dyn_hairs_06_02}
    componentAppliedLinearAccClamp=<2000 2000 2000>
    componentLinearAccScale=<2 2 2>
    componentLinearVelScale=<2 2 2>
    gravityOverride {z=-2000}
    physicsBodyDefinitions [{
      boundBone {boneName=dyn_hairs_06_02}
      boxExtents=<25 25 15>
      constraintSetup {angularLimitsMax=<0 1 1>, angularLimitsMin=<0 -1 -1>, angularTarget=<1 0 0>}
      localJointOffset=<7 0 0>
     }]
   }
   n26 = Anim Dynamics - Bone: dyn_hairs_18!(ComponentPose=@n25, Alpha=1, GravityOverride:Vector="0,0,-980", ExternalForce="0,0,0") -> (Pose) {
    angularDampingOverride=1
    angularSpringConstant=20000
    bAngularSpring=true
    bOverrideAngularDamping=true
    bUseGravityOverride=true
    gravityScale=0.35
    simulationSpace=RootRelative
    alphaBoolBlend {blendOption=Linear}
    boundBone {boneName=dyn_hairs_18}
    componentLinearAccScale=<1 1 1>
    componentLinearVelScale=<1 1 1>
    gravityOverride {z=-980}
    physicsBodyDefinitions [{
      boundBone {boneName=dyn_hairs_18}
      boxExtents=<20 30 30>
      constraintSetup {angularLimitsMax=<0 0 20>, angularLimitsMin=<0 -30 -20>, angularTarget=<1 0 0>}
      localJointOffset=<2 0 0>
     }]
   }
   n27 = Anim Dynamics - Bone: dyn_hairs_19!(ComponentPose=@n26, Alpha=1, GravityOverride:Vector="0,0,-980", ExternalForce="0,0,0") -> (Pose) {
    angularDampingOverride=1
    angularSpringConstant=20000
    bAngularSpring=true
    bOverrideAngularDamping=true
    bUseGravityOverride=true
    gravityScale=0.35
    simulationSpace=RootRelative
    alphaBoolBlend {blendOption=Linear}
    boundBone {boneName=dyn_hairs_19}
    componentLinearAccScale=<1 1 1>
    componentLinearVelScale=<1 1 1>
    gravityOverride {z=-980}
    physicsBodyDefinitions [{
      boundBone {boneName=dyn_hairs_19}
      boxExtents=<20 30 30>
      constraintSetup {angularLimitsMax=<0 0 20>, angularLimitsMin=<0 -30 -20>, angularTarget=<1 0 0>}
      localJointOffset=<2 0 0>
     }]
   }
   n28 = Anim Dynamics - Bone: dyn_hairs_20!(ComponentPose=@n27, Alpha=1, GravityOverride:Vector="0,0,-980", ExternalForce="0,0,0") -> (Pose) {
    angularDampingOverride=1
    angularSpringConstant=20000
    bAngularSpring=true
    bOverrideAngularDamping=true
    bUseGravityOverride=true
    gravityScale=0.35
    simulationSpace=RootRelative
    alphaBoolBlend {blendOption=Linear}
    boundBone {boneName=dyn_hairs_20}
    componentLinearAccScale=<1 1 1>
    componentLinearVelScale=<1 1 1>
    gravityOverride {z=-980}
    physicsBodyDefinitions [{
      boundBone {boneName=dyn_hairs_20}
      boxExtents=<20 30 30>
      constraintSetup {angularLimitsMax=<0 0 15>, angularLimitsMin=<0 -20 -15>, angularTarget=<1 0 0>}
      localJointOffset=<2 0 0>
     }]
   }
   n29 = Get bIsInFrontEnd()
   n30 = "Transform (Modify) Bone - Bone: dyn_hairs_11_02"!(ComponentPose=@n28, bAlphaBoolEnabled:bool=@n29, Translation="0,0,0", Rotation="0,0,5", Scale="1,1,1") -> (Pose) {
    alphaInputType=Bool
    rotationMode=BMM_Additive
    translationMode=BMM_Additive
    alphaBoolBlend {blendOption=Linear}
    boneToModify {boneName=dyn_hairs_11_02}
    rotation {roll=5}
   }
   n31 = "Transform (Modify) Bone - Bone: dyn_hairs_12_02"!(ComponentPose=@n30, bAlphaBoolEnabled:bool=@n29, Translation="0,0,0", Rotation="0,0,3", Scale="1,1,1") -> (Pose) {
    alphaInputType=Bool
    rotationMode=BMM_Additive
    translationMode=BMM_Additive
    alphaBoolBlend {blendOption=Linear}
    boneToModify {boneName=dyn_hairs_12_02}
    rotation {roll=3}
   }
   Output Pose(Result=Component To Local(ComponentPose=@n31)) {layerGroup=None}
  ]}
 EventGraph {nodes [
   Event BlueprintUpdateAnimation()
   > n1 = "Gravity Override State Function (Parameter Asset) (Nativized)"(
    CharacterPartAnimInstance=Self-Reference()
    GameTimeInSeconds=GetGameTimeInSeconds()
    Parameters=&/M_MED_ScubaDasher_Goggles_FaceAcc_Parameters
   )
   > Set Gravity Override(Gravity Override=@n1)
  ]}
]
variables [
 Gravity Override {type=Vector}
 Gravity Override Params {type=GravityOverrideParamsStruct}
]
